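; Small health foraged item that a pawn can consume or, when gathering is enabled, gather.
; Interaction is always offered; the first interaction passes only if the pawn is below max health and the parent check agrees.
; The second interaction follows the gather-enabled row value and plays the gather sound when allowed.
; Construction derives a deterministic pseudo-random mesh yaw from world position, varying placement without changing between builds and inflating patch size.
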(fn BlueprintCanInteract
  (params (InteractingPawn :type Fort%20Pawn%20Object%20Reference) (InteractionBeingAttempted :type EInteractionBeingAttempted%20Enum :default "FirstInteraction") (InteractionType :type TInteractionType%20Enum :default "IT_NoInteraction"))
  (return true))

(fn OnInteractAllowedToConsume
  (params (InteractingPawn :type Fort%20Pawn%20Object%20Reference) (InteractionBeingAttempted :type EInteractionBeingAttempted%20Enum :default "FirstInteraction"))
  (bind _row_gatherenabled (Variables|Gather|GetRow_GatherEnabled))
  (bind _returnvalue (Class|AthenaFunctionLibrary|RowtoBool _row_gatherenabled))
  (Utilities|IsValid InteractingPawn
    (:Is%20Valid
      ;[1 1 1] If second interact allow gather if enabled
      (switch Utilities|FlowControl|Switch|SwitchonEInteractionBeingAttempted InteractionBeingAttempted
        (:FirstInteraction
          (return (and (not (Math|Float|NearlyEqual(Float) (FortPawn|GetHealth InteractingPawn) (FortPawn|GetMaxHealth InteractingPawn))) (|CallParentFunction InteractingPawn "FirstInteraction" "IT_Simple"))))
        (:SecondInteraction
          (if _returnvalue
            (Audio|PlaySoundatLocation (Variables|Default|GetGatherSound) (Transformation|GetActorLocation) "0, 0, 0" 0.4))
          (return _returnvalue))))
    (:Is%20Not%20Valid
      (return false))))

(fn ConstructionScript
  (bind (_returnvalue_x _returnvalue_y _returnvalue_z) (Math|Vector|BreakVector (Transformation|GetWorldLocation (Variables|Default|GetStaticMesh))))
  ;[1 1 1] Give the mesh a random rot when it starts that makes them feel more randomly placed. The rot is psudo random so that it does not pick different values every build and cause large patch sizes.
  (|CallParentFunction)
  (Transformation|SetRelativeRotation (Variables|Default|GetStaticMesh) (Math|Rotator|MakeRotator (Variables|Default|GetRollOffset) 0.0 (Math|Conversions|ToFloat(Integer) (* (Math|Integer|%25%28Integer%29 (Math|Integer|int+int (Math|Float|Truncate _returnvalue_x) (Math|Float|Truncate _returnvalue_y) (Math|Float|Truncate _returnvalue_z)) 360) 12))) false false))
